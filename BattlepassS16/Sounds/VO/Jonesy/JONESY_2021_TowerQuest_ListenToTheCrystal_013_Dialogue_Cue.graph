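# Dialogue Sound Cue for Jonesy voice line 013 of the "Listen To The Crystal" tower quest in the Season 16 Battle Pass content.
# It has no root or sound nodes, so the cue itself references no audio.
# It carries only playback settings: the narrator VO sound class, the shared close-range event VO attenuation and a very large max distance.
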SoundCue /BattlepassS16/Sounds/VO/Jonesy/JONESY_2021_TowerQuest_ListenToTheCrystal_013_Dialogue_Cue
root=null
nodes []
properties {
 attenuationSettings=/Game/Athena/Sounds/Events/Attenuations/Events_VO_Close_Att
 maxDistance=2097152
 soundClassObject=/Game/Sounds/FortSoundClassesAndModes/Fort_VO_Narrator
 volumeMultiplier=1
}
