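# Sound cue for a Jonesy NPC voice line, named as listen line 010 of the 2021 Tower Quest in the Season 16 Battle Pass content.
# The cue has no root or sound nodes, so it references no audio as it stands.
# It carries only playback settings: a close-range event VO attenuation asset, the NPC voice sound class and a reduced volume multiplier.
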SoundCue /BattlepassS16/Sounds/VO/Jonesy/JONESY_2021_TowerQuest_Listen_010_Dialogue_Cue
root=null
nodes []
properties {
 attenuationSettings=/Game/Athena/Sounds/Events/Attenuations/Events_VO_Close_Att
 maxDistance=2097152
 soundClassObject=&Fort_VO_NPC
 volumeMultiplier=0.6667
}
